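# Sound cue for the Bad News weapon's aim-down-sights sound.
# A single wave player feeds one ADS wave straight to the output at reduced volume, with no randomization or mixing.
# It borrows pickup-style routing: the weapon pickup attenuation, the loot collect sound class and the PickupMaximum_Oldest concurrency set.
# It also stores single-instance, prevent-new concurrency override values, though the override flag is not shown enabled.
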SoundCue /MotherGameplay/Sounds/BadNews/WPN_Bad_News_ADS
root=WavePlayer_1
nodes [{id=WavePlayer_1, properties {soundWaveAssetPtr=&/WPN_Bad_News_ADS_01}}]
properties {
 attenuationSettings=&Fort_Weapon_Pickup_Att
 maxDistance=3500
 soundClassObject=&Fort_Loot_Collect
 volumeMultiplier=0.398
 concurrencyOverrides {maxCount=1, resolutionRule=PreventNew}
 concurrencySet [&PickupMaximum_Oldest]
}
