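# Sound cue for the charge sound of the Bad News weapon.
# A random node picks one of four equally weighted charge waves and feeds it through a modulator at the root.
# It uses the weapon's Unstable attenuation asset and a third-person ranged-fire sound class.
# Overridden concurrency allows one instance per owner, stopping the oldest when a new one starts.
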SoundCue /MotherGameplay/Sounds/BadNews/WPN_Bad_News_Charge_Cue
root=Modulator_0
nodes [
 {id=Modulator_0, children [Random_1]}
 {id=Random_1, children [WavePlayer_4, WavePlayer_5, WavePlayer_6, WavePlayer_20], properties {weights [1,1,1,1]}}
 {id=WavePlayer_20, properties {soundWaveAssetPtr=&/WPN_Bad_News_Charge_03}}
 {id=WavePlayer_4, properties {soundWaveAssetPtr=&/WPN_Bad_News_Charge_04}}
 {id=WavePlayer_5, properties {soundWaveAssetPtr=&/WPN_Bad_News_Charge_01}}
 {id=WavePlayer_6, properties {soundWaveAssetPtr=&/WPN_Bad_News_Charge_02}}
]
properties {
 attenuationSettings=&/WPN_Bad_News_Unstable_Att
 bOverrideConcurrency=true
 maxDistance=5800
 soundClassObject=&Fort_WeaponRanged_Fire_3P_PSM
 volumeMultiplier=1
 attenuationOverrides {falloffDistance=5000, attenuationShapeExtents {x=1000}}
 concurrencyOverrides {bLimitToOwner=true, maxCount=1, resolutionRule=StopOldest}
}
