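# Sound cue for the Bad News weapon's cooldown recharge foley.
# A single wave player feeds a modulator at the root.
# It plays in the weapon foley sound class with reload attenuation.
# A concurrency override limits it to one instance per owner, stopping the farthest before preventing new ones.
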SoundCue /MotherGameplay/Sounds/BadNews/WPN_Bad_News_Cooldown_Recharge_Cue
root=Modulator_0
nodes [
 {id=Modulator_0, children [WavePlayer_0]}
 {id=WavePlayer_0, properties {soundWaveAssetPtr=&/WPN_Bad_News_Cooldown_Recharge_01}}
]
properties {
 attenuationSettings=&Fort_Weapon_Reload_Att
 bOverrideConcurrency=true
 maxDistance=2250
 soundClassObject=&Fort_WeaponFoley
 volumeMultiplier=1.002
 concurrencyOverrides {bLimitToOwner=true, maxCount=1, resolutionRule=StopFarthestThenPreventNew}
}
