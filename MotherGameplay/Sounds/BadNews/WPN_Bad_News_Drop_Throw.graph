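# Sound cue for the drop or throw foley of the Bad News weapon.
# A single wave player feeds the throw sample into a modulator that raises pitch by a fixed amount, since its minimum and maximum match, so there is no pitch randomization.
# The cue boosts volume and uses the shared weapon-drop attenuation, the weapon foley sound class and a stop-oldest drop concurrency set.
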SoundCue /MotherGameplay/Sounds/BadNews/WPN_Bad_News_Drop_Throw
root=Modulator_1
nodes [
 {id=Modulator_1, children [WavePlayer_1], properties {pitchMax=1.1, pitchMin=1.1}}
 {id=WavePlayer_1, properties {soundWaveAssetPtr=&/WPN_Bad_News_Throw_01}}
]
properties {
 attenuationSettings=&Fort_Weapon_Drop_Att
 maxDistance=3000
 soundClassObject=&Fort_WeaponFoley
 volumeMultiplier=1.504
 concurrencyOverrides {resolutionRule=PreventNew}
 concurrencySet [&Weapon_Drop_StopOldest]
}
