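# WPN_Bad_News_Equip_Cue is the sound cue for equipping the Bad News weapon.
# A single wave player feeds the root directly with the WPN_Bad_News_Equip_01 wave, with no randomization or modulation.
# It plays through the weapon foley sound class, borrowing the weapon reload attenuation settings to limit its audible range.
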SoundCue /MotherGameplay/Sounds/BadNews/WPN_Bad_News_Equip_Cue
root=WavePlayer_3
nodes [{id=WavePlayer_3, properties {soundWaveAssetPtr=&/WPN_Bad_News_Equip_01}}]
properties {
 attenuationSettings=&Fort_Weapon_Reload_Att
 maxDistance=2250
 soundClassObject=&Fort_WeaponFoley
 volumeMultiplier=1.002
}
