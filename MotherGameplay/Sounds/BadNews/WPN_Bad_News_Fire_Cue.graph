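# Sound cue for Bad News weapon gunfire, layering close, medium and distant versions of the shot.
# A mixer blends the three layers, weighted toward close, each with its own distance attenuation and a sound class override: 3P for close, Distant for the other two.
# Each layer randomly picks one of four equally weighted waves and modulates it for variation.
# Shared global and local gunfire concurrency sets are assigned to the cue.
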SoundCue /MotherGameplay/Sounds/BadNews/WPN_Bad_News_Fire_Cue
root=Mixer_1
nodes [
 {
  id=Attenuation_10
  children [Modulator_14]
  properties {attenuationSettings=&Fort_Weapon_Projectile_Medium_Mono_Athena_Att}
 }
 {
  id=Attenuation_11
  children [Modulator_13]
  properties {attenuationSettings=&Fort_Weapon_Projectile_Distant_Mono_Athena_Att}
 }
 {id=Attenuation_9, children [Modulator_15], properties {attenuationSettings=&Fort_Weapon_Projectile_Close_Athena_Att}}
 {id=Mixer_1, children [SoundClass_7,SoundClass_6,SoundClass_0], properties {inputVolume [1,0.7,0.6]}}
 {id=Modulator_13, children [Random_6], properties {pitchMax=1.1, volumeMax=0.8, volumeMin=0.7}}
 {id=Modulator_14, children [Random_7], properties {pitchMax=1.1, volumeMax=0.8, volumeMin=0.7}}
 {id=Modulator_15, children [Random_4], properties {volumeMax=1.2, volumeMin=1}}
 {id=Random_4, children [WavePlayer_31, WavePlayer_32, WavePlayer_33, WavePlayer_34], properties {weights [1,1,1,1]}}
 {id=Random_6, children [WavePlayer_0,WavePlayer_1,WavePlayer_2,WavePlayer_3], properties {weights [1,1,1,1]}}
 {id=Random_7, children [WavePlayer_27, WavePlayer_28, WavePlayer_29, WavePlayer_30], properties {weights [1,1,1,1]}}
 {
  id=SoundClass_0
  children [Attenuation_11]
  properties {soundClassOverride=&Fort_WeaponRanged_Fire_Distant}
 }
 {
  id=SoundClass_6
  children [Attenuation_10]
  properties {soundClassOverride=&Fort_WeaponRanged_Fire_Distant}
 }
 {
  id=SoundClass_7
  children [Attenuation_9]
  properties {soundClassOverride=&Fort_WeaponRanged_Fire_3P_PSM}
 }
 {id=WavePlayer_0, properties {soundWaveAssetPtr=&/WPN_Bad_News_Fire_Dist_01}}
 {id=WavePlayer_1, properties {soundWaveAssetPtr=&/WPN_Bad_News_Fire_Dist_02}}
 {id=WavePlayer_2, properties {soundWaveAssetPtr=&/WPN_Bad_News_Fire_Dist_03}}
 {id=WavePlayer_27, properties {soundWaveAssetPtr=&/WPN_Bad_News_Fire_Med_01}}
 {id=WavePlayer_28, properties {soundWaveAssetPtr=&/WPN_Bad_News_Fire_Med_02}}
 {id=WavePlayer_29, properties {soundWaveAssetPtr=&/WPN_Bad_News_Fire_Med_03}}
 {id=WavePlayer_3, properties {soundWaveAssetPtr=&/WPN_Bad_News_Fire_Dist_04}}
 {id=WavePlayer_30, properties {soundWaveAssetPtr=&/WPN_Bad_News_Fire_Med_04}}
 {id=WavePlayer_31, properties {soundWaveAssetPtr=&/WPN_Bad_News_Fire_Close_01}}
 {id=WavePlayer_32, properties {soundWaveAssetPtr=&/WPN_Bad_News_Fire_Close_02}}
 {id=WavePlayer_33, properties {soundWaveAssetPtr=&/WPN_Bad_News_Fire_Close_03}}
 {id=WavePlayer_34, properties {soundWaveAssetPtr=&/WPN_Bad_News_Fire_Close_04}}
]
properties {
 maxDistance=27500
 volumeMultiplier=1
 attenuationOverrides {falloffDistance=5000, attenuationShapeExtents {x=1000}}
 concurrencyOverrides {bLimitToOwner=true, maxCount=3, resolutionRule=StopOldest}
 concurrencySet [&Weapon_Gunfire_Global_Conc, &Weapon_Gunfire_Local_Conc]
}
