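# Sound cue for the projectile explosion of the Bad News weapon.
# A Random node picks among three equally weighted explosion wave variations and passes the result through a Modulator.
# It plays in the weapons sound class with a natural-sound distance attenuation, including low-pass filter radii.
# A concurrency override keeps only a couple of instances, stopping the oldest when exceeded.
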SoundCue /MotherGameplay/Sounds/BadNews/WPN_Bad_News_Proj_Expl_Cue
root=Modulator_1
nodes [
 {id=Modulator_1, children [Random_0]}
 {id=Random_0, children [WavePlayer_0,WavePlayer_1,WavePlayer_2], properties {weights [1,1,1]}}
 {id=WavePlayer_0, properties {soundWaveAssetPtr=&/WPN_Bad_News_Proj_Expl_01}}
 {id=WavePlayer_1, properties {soundWaveAssetPtr=&/WPN_Bad_News_Proj_Expl_02}}
 {id=WavePlayer_2, properties {soundWaveAssetPtr=&/WPN_Bad_News_Proj_Expl_03}}
]
properties {
 attenuationSettings=&/WPN_Bad_News_Proj_Expl_Att
 bOverrideConcurrency=true
 maxDistance=8500
 soundClassObject=/Game/Sounds/FortSoundClassesAndModes/Fort_Weapons
 attenuationOverrides {distanceAlgorithm=NaturalSound, falloffDistance=3800, lPFRadiusMax=3000, lPFRadiusMin=2000, attenuationShapeExtents {x=2200}}
 concurrencyOverrides {maxCount=2, resolutionRule=StopOldest}
}
