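# Sound cue for the looping projectile sound of the Bad News weapon.
# A single looping wave player feeds a Modulator that slightly randomizes pitch.
# It plays in the ranged weapon sound class with the shared launcher in-air projectile attenuation asset.
# A concurrency override allows one instance at a time, stopping the farthest before refusing new ones.
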SoundCue /MotherGameplay/Sounds/BadNews/WPN_Bad_News_Proj_Loop_Cue
root=Modulator_0
nodes [
 {id=Modulator_0, children [WavePlayer_0], properties {pitchMax=1.1, pitchMin=0.9}}
 {id=WavePlayer_0, properties {bLooping=true, soundWaveAssetPtr=&/WPN_Bad_News_Proj_Loop_01}}
]
properties {
 attenuationSettings=/Game/Sounds/Attenuation/Fort_Weapon_Launcher_Proj_InAir_Default_Att
 bOverrideConcurrency=true
 maxDistance=12750
 soundClassObject=&Fort_WeaponRanged
 volumeMultiplier=0.7
 attenuationOverrides {
  bEnableListenerFocus=true
  falloffDistance=500
  focusAttackInterpSpeed=0.1
  focusAzimuth=60
  focusReleaseInterpSpeed=0.5
  nonFocusAzimuth=90
  nonFocusDistanceScale=0.7
  nonFocusVolumeAttenuation=0.5
  occlusionInterpolationTime=2
  occlusionTraceChannel=ECC_GameTraceChannel11
  occlusionVolumeAttenuation=0
  stereoSpread=1000
  attenuationShapeExtents {x=4000}
 }
 concurrencyOverrides {maxCount=1, resolutionRule=StopFarthestThenPreventNew, volumeScale=0.75}
}
