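# Sound cue for the zap sound of the Bad News weapon's projectile.
# A random node picks among four equally weighted zap waves, and a modulator on the output varies each play.
# It uses the trap sound class and attenuation with boosted volume, and its concurrency override prevents new instances once the limit is reached.
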SoundCue /MotherGameplay/Sounds/BadNews/WPN_Bad_News_Proj_Zap_Cue
root=Modulator_4
nodes [
 {id=Modulator_4, children [Random_0]}
 {
  id=Random_0
  children [WavePlayer_0,WavePlayer_2,WavePlayer_8,WavePlayer_9]
  properties {bShouldExcludeFromBranchCulling=true, weights [1,1,1,1]}
 }
 {id=WavePlayer_0, properties {soundWaveAssetPtr=&/WPN_Bad_News_Proj_Zap_01}}
 {id=WavePlayer_2, properties {soundWaveAssetPtr=&/WPN_Bad_News_Proj_Zap_02}}
 {id=WavePlayer_8, properties {soundWaveAssetPtr=&/WPN_Bad_News_Proj_Zap_03}}
 {id=WavePlayer_9, properties {soundWaveAssetPtr=&/WPN_Bad_News_Proj_Zap_04}}
]
properties {
 attenuationSettings=&Fort_Traps_Att
 bOverrideConcurrency=true
 maxDistance=6000
 soundClassObject=&Fort_Traps
 volumeMultiplier=2
 attenuationOverrides {falloffDistance=3200, attenuationShapeExtents {x=800}}
 concurrencyOverrides {resolutionRule=PreventNew}
}
