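# Sound cue for the Bad News weapon's unstable projectile sound.
# A Random node picks one of three equally weighted wave variations on each play.
# It uses its own attenuation asset, routes through a third-person ranged weapon fire sound class, and a concurrency rule caps it at one instance per owner, stopping the oldest.
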SoundCue /MotherGameplay/Sounds/BadNews/WPN_Bad_News_Unstable_Cue
root=Random_0
nodes [
 {id=Random_0, children [WavePlayer_0,WavePlayer_2,WavePlayer_3], properties {weights [1,1,1]}}
 {id=WavePlayer_0, properties {soundWaveAssetPtr=&/WPN_Bad_News_Proj_Unstable_01}}
 {id=WavePlayer_2, properties {soundWaveAssetPtr=&/WPN_Bad_News_Proj_Unstable_02}}
 {id=WavePlayer_3, properties {soundWaveAssetPtr=&/WPN_Bad_News_Proj_Unstable_03}}
]
properties {
 attenuationSettings=&/WPN_Bad_News_Unstable_Att
 maxDistance=5800
 soundClassObject=&Fort_WeaponRanged_Fire_3P_PSM
 volumeMultiplier=1
 concurrencySet [/Game/Sounds/Concurrency_Rules/OneMaximum_Owner_StopOldest]
}
